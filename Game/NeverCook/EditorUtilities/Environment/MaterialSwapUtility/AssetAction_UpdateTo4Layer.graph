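# Editor utility Blueprint in the MaterialSwapUtility tools, named as an asset action for updating environment assets to a 4-layer setup.
# Its UpdateTo4Layer function holds an Execute Python Command node that imports Environment.ConversionTools.asset_reskin, reloads it and calls Main().
# The conversion logic lives in that Python module, and the reload picks up script edits without an editor restart.
# The event graph is empty.
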EditorUtilityBlueprint /Game/NeverCook/EditorUtilities/Environment/MaterialSwapUtility/AssetAction_UpdateTo4Layer
graphs [
 EventGraph {links [], nodes []}
 UpdateTo4Layer {nodes [
   ExecutePythonCommand(PythonCommand="import Environment.ConversionTools.asset_reskin as AR; import importlib; importlib.reload(AR); AR.Main()")
   UpdateTo4Layer() {is=title K2Node_FunctionEntry}
  ]}
]
variables []
